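# Material function for the S30 player Superlevels look that restyles the base color and metallic of incoming material attributes.
# Base color is desaturated, contrast-adjusted, optionally hue-shifted via a static switch, then multiplied by a mask green channel and a tint; metallic gets a scalar multiplier.
# The result blends over the original only outside the TechArtMask RGB and TechArtDetailMask red regions, attenuated by a mask blue channel.
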MaterialFunction /Game/Characters/Player/TechArt/Superlevels/S30/MaterialFunctions/MF_SL_S30_Base
graph {nodes [
  n1 = GetMaterialAttributes(None=NamedRerouteUsage()) {attributeGetTypes [#1, #2]}
  n2 = MaterialFunctionCall!(
   In=Desaturation(None=@n1.BaseColor, Fraction=ScalarParameter() {defaultValue=1, group=SL_S30_Base, parameterName=SL_BaseColorDesaturation})
   Contrast=ScalarParameter() {defaultValue=-0.5, group=SL_S30_Base, parameterName=SL_BaseColorContrast}
  ) {materialFunction=/Engine/Functions/Engine_MaterialFunctions01/ImageAdjustment/CheapContrast_RGB}
  n3 = MaterialFunctionCall(Hue Shift Percentage=ScalarParameter() {group=SL_S30_Base, parameterName=SL_HueShift}, Texture=@n2.Result) {materialFunction=&HueShift}
  n4 = NamedRerouteUsage()
  n5 = Multiply(
   A=VectorParameter() {group=SL_S30_Base, parameterName=BaseColorTint, defaultValue=<rgba 0.0456 0.0482 0.0521 1>}.RGB
   B=Multiply(
    A=StaticSwitchParameter(True=@n3.Result, False=@n2.Result) {group=SL_S30_Base, parameterName=SL_UseHueShift}
    B=ComponentMask(None=@n4) {g=true}
   )
  )
  n6 = Multiply(A=ScalarParameter() {defaultValue=1, group=SL_S30_Base, parameterName=SL_MetalMultiplier}, B=@n1.Metallic)
  n7 = FunctionInput(Preview=TextureSample() {samplerType=SAMPLERTYPE_Masks, texture=&T_BlackMask}.RGB) {inputName=TechArtMask, sortPriority=1}
  n8 = Add(A=Add(A=ComponentMask(None=@n7) {r=true}, B=ComponentMask(None=@n7) {g=true}), B=ComponentMask(None=@n7) {b=true})
  n9 = ComponentMask(None=FunctionInput(Preview=TextureSample() {samplerType=SAMPLERTYPE_Masks, texture=&T_BlackMask}.RGB) {inputName=TechArtDetailMask, sortPriority=1}) {r=true}
  n10 = OneMinus(None=Saturate(None=Add(A=@n8, B=@n9)))
  n11 = OneMinus(None=ComponentMask(None=NamedRerouteUsage()) {b=true})
  FunctionOutput(None=BlendMaterialAttributes(
    A=NamedRerouteUsage()
    B=SetMaterialAttributes!(MaterialAttributes=@n1.MaterialAttributes, BaseColor=@n5, Metallic=@n6) {attributeSetTypes [#1, #2]}
    Alpha=LinearInterpolate(B=@n10, Alpha=@n11)
   ))
  NamedRerouteDeclaration(None=FunctionInput() {inputType=FunctionInput_MaterialAttributes}) {name=InMaterialAttributes, nodeColor=<rgba 1 0.3047 0 1>}
  NamedRerouteDeclaration(None=FunctionInput(Preview=Constant()) {inputName=Mask, sortPriority=1}) {name=MaskMap, nodeColor=<rgba 1 0 0.3047 1>}
 ]}
