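# Material function for the Season 30 player Superlevel style that turns masked areas into black leather, behind the UseBlackLeather static switch.
# Desaturates and tints base color, blends a tiled orange-peel noise normal, and scales roughness and metallic.
# An optional emissive layers projected hexagons with Fresnel and a time-driven sine sweep, using pre-skinned coordinates.
# A stepped TechArtMask channel blends the result over material attributes; cylindrical-UV experiments remain unconnected.
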MaterialFunction /Game/Characters/Player/TechArt/Superlevels/S30/MaterialFunctions/MF_SL_S30_BlackLeather
graph {nodes [
  n1 = NamedRerouteUsage()
  n2 = GetMaterialAttributes(None=NamedRerouteUsage()) {attributeGetTypes [
    #1
    #2
    #3
    #4
   ]}
  n3 = VectorParameter() {group=SL_S30_BlackLeather, parameterName=BL_ColorTint, defaultValue=<rgba 0.0552 0.0568 0.0573 1>}
  n4 = MaterialFunctionCall(
   In=Desaturation(None=@n2.BaseColor, Fraction=ScalarParameter() {defaultValue=1, group=SL_S30_BlackLeather, parameterName=BL_Desaturation})
   Contrast=ScalarParameter() {defaultValue=-0.5, group=SL_S30_BlackLeather, parameterName=BL_Contrast}
  ) {materialFunction=&CheapContrast}
  n5 = NamedRerouteUsage()
  n6 = TextureSample(UVs=Multiply(A=TextureCoordinate(), B=ScalarParameter() {defaultValue=5, group=SL_S30_BlackLeather, parameterName=LeatherNoiseTiling})) {samplerType=SAMPLERTYPE_Normal, texture=/Game/Global/Textures/Default/Noise/T_OrangePeel_N}
  n7 = ScalarParameter() {defaultValue=0.75, group=SL_S30_BlackLeather, parameterName=LeatherNormalIntensity}
  n8 = ScalarParameter() {defaultValue=1, group=SL_S30_BlackLeather, parameterName=BL_Roughness}
  n9 = SetMaterialAttributes!(
   MaterialAttributes=@n2.MaterialAttributes
   BaseColor=Multiply(A=Multiply(A=@n3.RGB, B=@n4.Result), B=ComponentMask(None=@n5) {g=true})
   Normal=MaterialFunctionCall(
    BaseNormal=@n2.Normal
    AdditionalNormal=MaterialFunctionCall(Normal=@n6.RGB, Flatness=@n7) {materialFunction=&FlattenNormal}.Result
   ) {materialFunction=&BlendAngleCorrectedNormals}.Result
   Roughness=Saturate(None=Multiply(A=@n2.Roughness, B=@n8))
   EmissiveColor=NamedRerouteUsage()
   Opacity=Constant()
   Metallic=Multiply(A=@n2.Metallic, B=ScalarParameter() {group=SL_S30_BlackLeather, parameterName=BL_Metalness})
  ) {attributeSetTypes [
    #2
    #1
    #3
    #5
    #6
    #4
   ]}
  n10 = Step(
   Y=ScalarParameter() {defaultValue=0.1, group=SL_S30_BlackLeather, parameterName=BL_ClippingMask}
   X=ChannelMaskParameter(None=FunctionInput() {inputName=TechArtMask, sortPriority=1}) {group=SL_S30_BlackLeather, parameterName=BlackLeatherMask}.R
  )
  FunctionOutput(None=StaticSwitchParameter(True=BlendMaterialAttributes(A=@n1, B=@n9, Alpha=@n10), False=NamedRerouteUsage()) {group=S30 Superleveling, parameterName=UseBlackLeather})
  MaterialFunctionCall!(BoundsScale) {materialFunction=/Game/Athena/SafeZone/S11/MF_CylindricalUVs}
  NamedRerouteDeclaration(None=FunctionInput() {inputType=FunctionInput_MaterialAttributes}) {name=InMaterialAttributes, nodeColor=<rgba 1 0.3047 0 1>}
  n11 = Multiply(A=Time(), B=ScalarParameter() {defaultValue=0.15, group=SL_S30_BlackLeather, parameterName=BL_SweepSpeed}) {constB=0.15}
  n12 = PreSkinnedPosition()
  n13 = VertexInterpolator(VS=@n12)
  n14 = Divide(
   A=ComponentMask(None=ShaderStageSwitch(PixelShader=@n13.PS, VertexShader=@n12)) {b=true}
   B=ScalarParameter() {defaultValue=150, group=SL_S30_BlackLeather, parameterName=BL_SweepSize}
  ) {constB=-125}
  n15 = ConstantBiasScale(None=Sine(None=Add(A=@n11, B=@n14)))
  NamedRerouteDeclaration(None=Power(Base=Saturate(None=@n15), Exp=ScalarParameter() {defaultValue=2, group=SL_S30_BlackLeather, parameterName=BL_SweepSmoothness}) {constExponent=8}) {name=GradientMask, nodeColor=<rgba 0 1 0.3906 1>}
  NamedRerouteDeclaration(None=FunctionInput(Preview=Constant() {r=1}) {inputName=Mask, sortPriority=1}) {name=MaskMap, nodeColor=<rgba 1 0 0.3047 1>}
  n16 = VectorParameter() {group=SL_S30_BlackLeather, parameterName=HexColors, defaultValue=<rgba 0 0.592 1.5 1>}
  n17 = MaterialFunctionCall(
   TextureObject=TextureObject() {texture=/Game/Characters/Player/TechArt/Superlevels/S30/Textures/T_S30_Superlevel_T1_Hexagons}
   TextureSize=ScalarParameter() {defaultValue=30, group=SL_S30_BlackLeather, parameterName=HexTiling}
   WorldPosition=VertexInterpolator(VS=LocalPosition() {localOrigin=InstancePreSkinning}.XYZ).PS
   World Space Normal=VertexInterpolator(VS=PreSkinnedNormal()).PS
  ) {materialFunction=&WorldAlignedTexture}
  n18 = OneMinus(None=ComponentMask(None=@n17.XYZ Texture) {g=true})
  n19 = ComponentMask(None=@n17.XYZ Texture) {r=true}
  n20 = NamedRerouteUsage()
  n21 = Multiply(A=@n19, B=Multiply(A=@n20, B=@n20))
  n22 = Sine(None=Add(A=Add(A=@n18) {constB=0.35}, B=Multiply(A=@n21) {constB=0.3}))
  n23 = Saturate(None=Power(Base=@n22) {constExponent=32})
  n24 = Multiply(A=Multiply(A=NamedRerouteUsage(), B=@n23)) {constB=3}
  n25 = NamedRerouteUsage()
  n26 = Fresnel(ExponentIn=Subtract(A=ScalarParameter() {defaultValue=3.5, group=SL_S30_BlackLeather, parameterName=HexFresnelPower}, B=Multiply(A=@n25) {constB=0.5})) {baseReflectFraction=0, exponent=3}
  n27 = Multiply(A=Multiply(A=Add(A=@n24, B=@n23), B=Multiply(A=@n26) {constB=2}), B=LinearInterpolate(Alpha=@n25) {constA=0.75})
  n28 = Multiply(A=@n26) {constB=0.25}
  n29 = NamedRerouteUsage()
  n30 = Subtract(
   A=MaterialFunctionCall(In=ComponentMask(None=@n29) {g=true}, Contrast=Constant() {r=1}) {materialFunction=&CheapContrast}.Result
   B=ComponentMask(None=NamedRerouteUsage()) {b=true}
  )
  n31 = Multiply(A=ComponentMask(None=VertexNormalWS()) {b=true}) {constB=1.5}
  n32 = Multiply(A=Multiply(A=Add(A=@n27, B=@n28), B=Saturate(None=@n30)), B=Saturate(None=OneMinus(None=@n31)))
  NamedRerouteDeclaration(None=StaticSwitchParameter(True=Multiply(A=@n16.RGB, B=@n32), False=Constant()) {defaultValue=true, group=SL_S30_BlackLeather, parameterName=UseHexFresnel}) {name=OutEmissive, nodeColor=<rgba 0 0.5078 1 1>}
  NamedRerouteDeclaration(None=FunctionInput(Preview=Constant()) {inputName=SL2Mask, sortPriority=1}) {name=SL2Mask, nodeColor=<rgba 1 0.7031 0 1>}
  n33 = Time()
  Sine(None=Subtract(A=Multiply(A=@n33) {constB=0.05}, B=@n19))
  SmoothStep(Value=@n22) {constMin=0.9}
  Step(X=@n22) {constY=0.95}
  n34 = VertexInterpolator(VS=LocalPosition() {localOrigin=InstancePreSkinning}.XYZ)
  n35 = ComponentMask(None=@n34.PS) {g=true, r=true}
  n36 = NamedRerouteDeclaration(None=ComponentMask(None=PreSkinnedLocalBounds().Extents) {b=true}) {name=SL_PreskinnedLocalBoundsZ, nodeColor=<rgba 0 1 0.3203 1>}
  n37 = AppendVector(
   A=Multiply(
    A=MaterialFunctionCall(Vector or UVs=@n35) {materialFunction=&VectorToRadialValue}.Vector Converted to Angle
    B=ScalarParameter() {defaultValue=10, group=00_SuperLevels_T1, parameterName=SL_Radial Tiling}
   ) {constB=2}
   B=ComponentMask(None=Divide(A=@n34.PS, B=@n36)) {b=true}
  )
  n38 = AppendVector(A=Constant() {r=1}, B=ScalarParameter() {defaultValue=10, group=00_SuperLevels_T1, parameterName=CylindricalUVs_VerticalScale})
  TextureSample(UVs=NamedRerouteDeclaration(None=Multiply(A=@n37, B=@n38)) {name=CylUV, nodeColor=<rgba 0.6875 0 1 1>}) {texture=&T_S30_Superlevel_T1_Hexagons}
  VectorParameter() {group=00_SuperLevels_T1, parameterName=SL_CylindricalProj_Tiling, channelNames {a="", b="", g=V, r=U}, defaultValue=<rgba 1 200 0 1>}
 ]}
