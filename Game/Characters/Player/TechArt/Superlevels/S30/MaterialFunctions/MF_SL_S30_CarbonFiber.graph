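# Material function for the Season 30 superlevel carbon fiber style on player characters, switchable by a static parameter and blended into material attributes through a thresholded channel mask.
# Carbon fiber color and normals use UVs or optional pre-skinned triplanar projection, with tint and fake highlights.
# A hexagon texture varies roughness.
# Emissive combines Fresnel color with noise-perturbed thin-film iridescence, while panning noise and a sine sweep supply animated masks and normal-rotation amounts.
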MaterialFunction /Game/Characters/Player/TechArt/Superlevels/S30/MaterialFunctions/MF_SL_S30_CarbonFiber
graph {nodes [
  n1 = Time()
  n2 = ScalarParameter() {defaultValue=0.15, parameterName=CF_SweepSpeed}
  n3 = PreSkinnedPosition()
  n4 = ComponentMask(None=ShaderStageSwitch(PixelShader=VertexInterpolator(VS=@n3).PS, VertexShader=@n3)) {b=true}
  n5 = ScalarParameter() {defaultValue=300, parameterName=CF_SweepSize}
  n6 = Sine(None=Add(A=Multiply(A=@n1, B=@n2) {constB=0.15}, B=Divide(A=@n4, B=@n5) {constB=-125}))
  Abs(None=@n6)
  n7 = VertexInterpolator(VS=LocalPosition() {localOrigin=InstancePreSkinning}.XYZ)
  Add(
   A=@n7.PS
   B=RotateAboutAxis(
    NormalizedRotationAxis=Constant3Vector() {constant=<rgba 0 0 1 0>}
    RotationAngle=Constant() {r=0.25}
    PivotPoint=Constant3Vector()
    Position=@n7.PS
   )
  )
  n8 = NamedRerouteUsage()
  n9 = NamedRerouteUsage()
  n10 = FunctionInput() {inputName=TechArtMask, sortPriority=1}
  n11 = NamedRerouteDeclaration(None=Step(
    Y=ScalarParameter() {defaultValue=0.1, group=SL_S30_CarbonFiber, parameterName=CF_ClippingMas}
    X=ChannelMaskParameter(None=@n10) {group=SL_S30_CarbonFiber, parameterName=CarbonFiberMask}.R
   )) {name=CarbonFiberMask, nodeColor=<rgba 1 0.7969 0 1>}
  FunctionOutput(None=StaticSwitchParameter(True=BlendMaterialAttributes(A=@n8, B=@n9, Alpha=@n11), False=NamedRerouteUsage()) {group=S30 Superleveling, parameterName=UseCarbonFiber})
  n12 = NamedRerouteUsage()
  n13 = LocalPosition() {localOrigin=InstancePreSkinning}
  n14 = VertexInterpolator(VS=Multiply(A=ComponentMask(None=@n13.XYZ) {b=true, r=true}) {constB=0.005})
  n15 = Multiply(A=@n14.PS, B=ComponentMask(None=VectorParameter() {parameterName=ScaleNoiseTiling, defaultValue=<rgba 0.75 1 0 1>}.RGB) {g=true, r=true}) {constB=2}
  n16 = VectorParameter() {parameterName=ScaleNoiseSpeed, defaultValue=<rgba 0 0.075 0 1>}
  n17 = TextureSample(UVs=Panner(Coordinate=Multiply(A=@n15) {constB=2.5}, Speed=ComponentMask(None=@n16.RGB) {g=true, r=true}) {speedY=0.075}) {
   samplerType=SAMPLERTYPE_LinearColor
   texture=/Game/Characters/Player/Female/Medium/Bodies/F_MED_ReptilianOcean_Sleek/Textures/T_ReptilianOcean_Sleek_Pattern_Noise
  }
  Multiply(A=@n12, B=LinearInterpolate(Alpha=NamedRerouteDeclaration(None=@n17.R) {name=PanningUpNoise, nodeColor=<rgba 0.8516 1 0 1>}))
  NamedRerouteDeclaration(None=FunctionInput() {inputType=FunctionInput_MaterialAttributes}) {name=InMaterialAttributes, nodeColor=<rgba 1 0.3047 0 1>}
  n18 = Constant()
  n19 = SetMaterialAttributes!(
   MaterialAttributes
   BaseColor=NamedRerouteUsage()
   Metallic=@n18
   Specular=ScalarParameter() {defaultValue=0.5, group=SL_S30_CarbonFiber, parameterName=CF_Specular}
   Roughness=NamedRerouteUsage()
   Normal=NamedRerouteUsage()
   ShadingModel=ShadingModel()
   CustomData0
   CustomData1
   Opacity=@n18
   EmissiveColor=NamedRerouteUsage()
  ) {attributeSetTypes [
    #1
    #2
    #3
    #4
    #5
    #6
    #7
    #8
    #9
    #10
   ]}
  NamedRerouteDeclaration(None=@n19) {name=OutCarbonFiber, nodeColor=<rgba 1 0 0.0469 1>}
  n20 = ComponentMask(None=NamedRerouteUsage()) {g=true}
  n21 = ComponentMask(None=NamedRerouteUsage()) {b=true, g=true, r=true}
  n22 = LinearInterpolate(Alpha=OneMinus(None=NamedRerouteUsage())) {constA=0.2}
  n23 = Multiply(
   A=Multiply(
    A=VectorParameter() {group=SL_S30_CarbonFiber, parameterName=CarbonFiberTint, defaultValue=<rgba 0.0412 0.0447 0.0521 1>}.RGB
    B=Add(A=@n21) {constB=0.2}
   )
   B=@n22
  )
  n24 = MaterialFunctionCall(BaseNormal=NamedRerouteUsage(), AdditionalNormal=NamedRerouteUsage()) {materialFunction=&BlendAngleCorrectedNormals}
  n25 = OneMinus(None=Fresnel(Normal=Transform(None=@n24.Result)) {baseReflectFraction=0, exponent=0.05})
  n26 = ScalarParameter() {defaultValue=0.25, group=SL_S30_CarbonFiber, parameterName=CF_FakeHighlightAmount}
  n27 = Add(A=LinearInterpolate(A=Multiply(A=@n20, B=@n23), Alpha=NamedRerouteUsage()) {constB=0}, B=Multiply(A=@n22, B=Multiply(A=@n25, B=@n26)))
  NamedRerouteDeclaration(None=@n27) {name=OutBaseColor, nodeColor=<rgba 0.0312 0 1 1>}
  n28 = TextureObjectParameter() {parameterName=Param_1, texture=/Game/Weapons/WeaponSkins/Wraps/Textures/T_Wrap_CarbonFiber_D}
  n29 = NamedRerouteUsage()
  n30 = StaticBool() {value=true}
  n31 = VertexInterpolator(VS=PreSkinnedNormal())
  n32 = ScalarParameter() {defaultValue=1, group=SL_S30_CarbonFiber, parameterName=CF_TriplanarBlend}
  n33 = ScalarParameter() {defaultValue=15, group=SL_S30_CarbonFiber, parameterName=CarbonFiberTiling}
  n34 = NamedRerouteDeclaration(None=Multiply(A=TextureCoordinate(), B=NamedRerouteDeclaration(None=@n33) {name=CF_Tiling, nodeColor=<rgba 1 0 0.2812 1>})) {name=UVs, nodeColor=<rgba 0.2891 1 0 1>}
  NamedRerouteDeclaration(None=StaticSwitchParameter(
    True=MaterialFunctionCall(
     TextureObject=@n28
     TextureSize=@n29
     WorldPosition=@n7.PS
     Export Float 4=@n30
     World Space Normal=@n31.PS
     ProjectionTransitionContrast=@n32
    ) {materialFunction=&WorldAlignedTexture}.XYZ Texture
    False=TextureSample(UVs=@n34) {texture=&T_Wrap_CarbonFiber_D}.RGBA
   ) {group=SL_S30_CarbonFiber, parameterName=CF_UseTriplannarMapping}) {name=InBaseColor, nodeColor=<rgba 0.3594 0 1 1>}
  NamedRerouteDeclaration(None=Power(Base=Saturate(None=@n6), Exp=ScalarParameter() {defaultValue=2, parameterName=CF_SweepSmoothness}) {constExponent=8}) {name=GradientMask, nodeColor=<rgba 0 1 0.3906 1>}
  n35 = DotProduct(A=VertexNormalWS(), B=Constant3Vector() {constant=<rgba 0 0 -1 0>})
  n36 = Saturate(None=Power(Base=OneMinus(None=@n35)) {constExponent=4})
  n37 = OneMinus(None=NamedRerouteUsage())
  n38 = OneMinus(None=Fresnel() {baseReflectFraction=0, exponent=1})
  n39 = Multiply(
   A=ScalarParameter() {defaultValue=1.5, parameterName=Param_4}
   B=Power(Base=SmoothStep(Value=@n38), Exp=ScalarParameter() {defaultValue=2, group=SL_S30_CarbonFiber, parameterName=CF_ExtraThinFilmMaskExponent}) {constExponent=4}
  )
  n40 = TextureSample(UVs=Multiply(A=TextureCoordinate()) {constB=3}) {
   samplerType=SAMPLERTYPE_LinearColor
   texture=&T_ReptilianOcean_Sleek_Pattern_Noise
  }
  n41 = Add(A=NamedRerouteUsage(), B=Multiply(A=@n40.R) {constB=0.5})
  n42 = Power(Base=@n38, Exp=Constant() {r=2})
  n43 = Add(
   A=LinearInterpolate(B=Saturate(None=@n41), Alpha=Saturate(None=@n42))
   B=ScalarParameter() {defaultValue=0.95, group=SL_S30_CarbonFiber, parameterName=CF_ExtraThinFilmOffset}
  ) {constB=0.5}
  n44 = ScalarParameter() {defaultValue=0.25, group=SL_S30_CarbonFiber, parameterName=CF_ExtraThinFilmScale}
  n45 = TextureSample(UVs=Sine(None=Multiply(A=@n43, B=@n44) {constB=1.5})) {texture=/Game/Global/Textures/Default/Thinfilm/T_ThinFilm_Spectrum_COLOR}
  n46 = NamedRerouteDeclaration(None=Multiply(A=Multiply(A=@n37, B=@n37), B=Multiply(A=@n39, B=@n45.RGB))) {name=ExtraThinFilm, nodeColor=<rgba 1 0.1172 0 1>}
  n47 = Add(A=Max(A=NamedRerouteUsage()) {constB=0.1}, B=Multiply(A=@n25) {constB=0.5})
  n48 = Multiply(A=ComponentMask(None=NamedRerouteUsage()) {g=true}, B=Multiply(A=Multiply(A=@n46, B=@n47)) {constB=0.4})
  NamedRerouteDeclaration(None=LinearInterpolate(
    A=Multiply(A=@n36, B=@n48)
    B=VectorParameter() {group=SL_S30_CarbonFiber, parameterName=CF_FresnelColor, defaultValue=<rgba 0.1979 0.6786 2 1>}.RGB
    Alpha=NamedRerouteUsage()
   )) {name=OutEmissive, nodeColor=<rgba 0.9688 0 1 1>}
  n49 = ComponentMask(None=VectorParameter() {parameterName=ScalesExtraSweepScale, defaultValue=<rgba 1 0.5 0 1>}.RGB) {g=true, r=true}
  n50 = TextureSample(UVs=Panner(Coordinate=Multiply(A=@n49, B=@n14.PS) {constB=4}) {speedY=-0.015}) {
   samplerType=SAMPLERTYPE_LinearColor
   texture=&T_ReptilianOcean_Sleek_Pattern_Noise
  }
  NamedRerouteDeclaration(None=@n50.R) {name=PanningUpNoise2, nodeColor=<rgba 0.8516 1 0 1>}
  n51 = ScalarParameter() {defaultValue=4, group=SL_S30_CarbonFiber, parameterName=CF_FresnelPower}
  n52 = LinearInterpolate(A=Transform(None=NamedRerouteUsage()), B=Transform(None=NamedRerouteUsage())) {constAlpha=0.2}
  NamedRerouteDeclaration(None=Saturate(None=Fresnel(ExponentIn=@n51, Normal=@n52) {baseReflectFraction=0})) {name=Fresnel, nodeColor=<rgba 0.0781 0 1 1>}
  n53 = TextureSample(UVs=Multiply(A=TextureCoordinate())) {samplerSource=SSM_Wrap_WorldGroupSettings, texture=/Game/Characters/Player/TechArt/Superlevels/S30/Textures/T_S30_Superlevel_T1_Hexagons}
  n54 = Power(Base=MaterialFunctionCall(In=@n53.B, Contrast=Constant()) {materialFunction=&CheapContrast}.Result) {constExponent=1}
  NamedRerouteDeclaration(None=LinearInterpolate(A=ScalarParameter() {defaultValue=0.2, group=SL_S30_CarbonFiber, parameterName=CF_Roughness}, Alpha=Saturate(None=@n54))) {name=Roughness, nodeColor=<rgba 1 0 0.2109 1>}
  NamedRerouteDeclaration(None=@n27) {name=BaseColor, nodeColor=<rgba 1 0 0.2109 1>}
  n55 = LinearInterpolate(Alpha=SmoothStep(Value=@n50.G) {constMin=0.1}) {constA=-0.2}
  n56 = ScalarParameter() {defaultValue=0.1, parameterName=ScalesSecondaryNormalInfluence}
  n57 = ScalarParameter() {defaultValue=0.2, parameterName=ScalesSweepNormalRotateAmount}
  NamedRerouteDeclaration(None=Add(A=Multiply(A=@n55, B=@n56) {constB=0.5}, B=Multiply(A=@n12, B=@n57) {constB=0.15})) {name=NormalsRotationAmount, nodeColor=<rgba 0 1 0.5547 1>}
  NamedRerouteDeclaration(None=GetMaterialAttributes(None=NamedRerouteUsage()) {attributeGetTypes [#4, #5, #1]}.Normal) {name=InTangentSpaceNormals, nodeColor=<rgba 0 0.625 1 1>}
  n58 = StaticSwitchParameter(
   True=MaterialFunctionCall!(
    TextureObject=TextureObjectParameter() {parameterName=Param, samplerType=SAMPLERTYPE_Normal, texture=/Game/Weapons/WeaponSkins/Wraps/Textures/T_Wrap_CarbonFiber_N}
    TextureSize=NamedRerouteUsage()
    PreskinnedLocalNormal=@n31.PS
    PreskinnedLocalPosition=@n7.PS
    Use High Quality Normals
   ) {materialFunction=/Game/Global/MaterialFunctions/MF_PreskinnedWorldAlignedTextureNormal}.XYZ Texture
   False=TextureSample(UVs=NamedRerouteUsage()) {samplerType=SAMPLERTYPE_Normal, texture=&T_Wrap_CarbonFiber_N}.RGB
  ) {defaultValue=true, group=SL_S30_CarbonFiber, parameterName=CF_UseTriplannarMapping}
  n59 = Constant3Vector() {constant=<rgba 0 0 1 0>}
  n60 = Fresnel() {baseReflectFraction=0, exponent=2}
  n61 = ScalarParameter() {defaultValue=0.5, parameterName=CarbonFiberNormalIntensity}
  NamedRerouteDeclaration(None=MaterialFunctionCall(Normal=LinearInterpolate(A=@n58, B=@n59, Alpha=@n60), Flatness=Multiply(A=@n61) {constB=-1}) {materialFunction=&FlattenNormal}.Result) {name=CarbonFiberNormals, nodeColor=<rgba 1 0.5859 0 1>}
  NamedRerouteDeclaration(None=FunctionInput(Preview=Constant() {r=1}) {inputName=Mask, sortPriority=1}) {name=MaskMap, nodeColor=<rgba 1 0 0.3047 1>}
  n62 = NamedRerouteUsage()
  n63 = Constant3Vector() {constant=<rgba 1 1 0 0>}
  NamedRerouteDeclaration(None=Saturate(None=DotProduct(A=@n62, B=@n63))) {name=FakeAO, nodeColor=<rgba 0.7812 1 0 1>}
  n64 = NamedRerouteUsage()
  n65 = Constant3Vector() {constant=<rgba 0 0 -1 0>}
  n66 = CrossProduct(
   A=Normalize(VectorInput=Transform(None=@n65) {transformSourceType=TRANSFORMSOURCE_World, transformType=TRANSFORM_Tangent})
   B=Transform(None=VertexNormalWS()) {transformSourceType=TRANSFORMSOURCE_World, transformType=TRANSFORM_Tangent}
  )
  n67 = NamedRerouteUsage()
  n68 = Constant()
  NamedRerouteDeclaration(None=Add(A=@n64, B=RotateAboutAxis(NormalizedRotationAxis=@n66, RotationAngle=@n67, PivotPoint=@n68, Position=@n64))) {name=InvertedNormals, nodeColor=<rgba 1 0 0.1406 1>}
  NamedRerouteUsage()
  NamedRerouteUsage()
  PixelNormalWS()
 ]}
