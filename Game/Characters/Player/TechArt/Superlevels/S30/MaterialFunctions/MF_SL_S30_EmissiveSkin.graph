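# Material function for the S30 superlevel player styles that adds a tinted emissive skin effect, gated by the UseEmissiveSkin static switch.
# UseEmissiveChannel selects between tinting existing emissive, weighted by its desaturated, contrast-adjusted, squared value, and tinting a green-minus-red mask.
# The mask path uses MF_Flicker to vary power and intensity, making the glow grow and pulse.
# A blue-channel attribute blend is present but does not feed the output.
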MaterialFunction /Game/Characters/Player/TechArt/Superlevels/S30/MaterialFunctions/MF_SL_S30_EmissiveSkin
graph {nodes [
  n1 = GetMaterialAttributes(None=NamedRerouteUsage()) {attributeGetTypes [#1]}
  n2 = VectorParameter() {group=SL_S30_EmissiveSkin, parameterName=ES_ColorTint, defaultValue=<rgba 0 0.8773 1 1>}
  n3 = MaterialFunctionCall(
   In=Desaturation(None=@n1.EmissiveColor, Fraction=ScalarParameter() {defaultValue=1, group=SL_S30_EmissiveSkin, parameterName=ES_Desaturation})
   Contrast=ScalarParameter() {group=SL_S30_EmissiveSkin, parameterName=ES_Contrast}
  ) {materialFunction=&CheapContrast}
  n4 = Multiply(A=@n3.Result, B=@n3.Result)
  n5 = NamedRerouteUsage()
  n6 = Subtract(A=ComponentMask(None=@n5) {g=true}, B=ComponentMask(None=@n5) {r=true})
  n7 = ScalarParameter() {defaultValue=1, group=SL_S30_EmissiveSkin, parameterName=ES_Power}
  n8 = Multiply(A=MaterialFunctionCall!(Time Mult=Constant() {r=0.1}) {materialFunction=/Game/Effects/Fort_Effects/MaterialFunctions/MF_Flicker}.Raw Flicker) {constB=0.125}
  n9 = Multiply(
   A=Power(Base=Saturate(None=@n6), Exp=Add(A=@n7, B=@n8))
   B=Add(A=ScalarParameter() {defaultValue=0.5, group=SL_S30_EmissiveSkin, parameterName=ES_Intensity}, B=Abs(None=@n8))
  ) {desc=Make it grow and pulse}
  n10 = SetMaterialAttributes!(
   MaterialAttributes=@n1.MaterialAttributes
   EmissiveColor=StaticSwitchParameter(True=LinearInterpolate(A=@n1.EmissiveColor, B=@n2.RGB, Alpha=@n4), False=Multiply(A=@n2.RGB, B=@n9)) {group=SL_S30_EmissiveSkin, parameterName=UseEmissiveChannel}
  ) {attributeSetTypes [#1]}
  BlendMaterialAttributes(A=NamedRerouteUsage(), B=@n10, Alpha=ComponentMask(None=NamedRerouteUsage()) {b=true})
  FunctionOutput(None=StaticSwitchParameter(True=@n10, False=NamedRerouteUsage()) {group=S30 Superleveling, parameterName=UseEmissiveSkin})
  NamedRerouteDeclaration(None=FunctionInput() {inputType=FunctionInput_MaterialAttributes}) {name=InMaterialAttributes, nodeColor=<rgba 1 0.3047 0 1>}
  NamedRerouteDeclaration(None=FunctionInput(Preview=Constant() {r=1}) {inputName=Mask, sortPriority=1}) {name=MaskMap, nodeColor=<rgba 1 0 0.3047 1>}
 ]}
