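# Material function for the Season 30 superlevel character styles that adds a switchable extra-detail layer to incoming material attributes.
# It turns desaturated, contrast-adjusted, optionally inverted base color, masked by a green channel, plus a thresholded Fresnel outline into tinted HDR emissive, with quality and ray-tracing switches.
# The layer zeroes base color and metallic and derives roughness from a hexagon texture.
# A smoothstepped TechArtMask channel blends it over the input.
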MaterialFunction /Game/Characters/Player/TechArt/Superlevels/S30/MaterialFunctions/MF_SL_S30_ExtraDetail
graph {nodes [
  n1 = NamedRerouteUsage()
  n2 = GetMaterialAttributes(None=NamedRerouteUsage()) {attributeGetTypes [#1, #2]}
  n3 = Constant()
  n4 = MaterialFunctionCall(
   In=Desaturation(None=@n2.BaseColor, Fraction=ScalarParameter() {defaultValue=1, group=SL_S30_ExtraDetail, parameterName=ED_BaseTextureDesaturation})
   Contrast=ScalarParameter() {defaultValue=-1, group=SL_S30_ExtraDetail, parameterName=ED_BaseTextureContrast}
  ) {materialFunction=&CheapContrast}
  n5 = Multiply(
   A=LinearInterpolate(A=@n4.Result, B=OneMinus(None=@n4.Result), Alpha=ScalarParameter() {group=SL_S30_ExtraDetail, parameterName=ED_InvertBaseColor})
   B=ComponentMask(None=NamedRerouteUsage()) {g=true}
  )
  n6 = ScalarParameter() {defaultValue=1, group=SL_S30_ExtraDetail, parameterName=ED_OutlinePower}
  n7 = ScalarParameter() {defaultValue=1, group=SL_S30_ExtraDetail, parameterName=ED_OutlineIntensity}
  n8 = ScalarParameter() {defaultValue=5, group=SL_S30_ExtraDetail, parameterName=ED_LumenContribution}
  n9 = Multiply(
   A=SmoothStep(
    Min=ScalarParameter() {defaultValue=0.1, group=SL_S30_ExtraDetail, parameterName=ED_OutlineThreshhold}
    Value=Fresnel(ExponentIn=@n6) {baseReflectFraction=0, exponent=0.5}
   ) {constMin=0.1}
   B=RayTracingQualitySwitch(Normal=@n7, RayTraced=Multiply(A=@n7, B=@n8))
  )
  n10 = Add(
   A=@n2.EmissiveColor
   B=Multiply(A=Add(A=@n5, B=@n9), B=VectorParameter() {group=SL_S30_ExtraDetail, parameterName=ED_ColorTint, defaultValue=<rgba 0 0.1634 1 1>}.RGB)
  )
  n11 = Constant() {r=100}
  n12 = Constant() {r=0.75}
  n13 = SetMaterialAttributes!(
   MaterialAttributes=@n2.MaterialAttributes
   BaseColor=@n3
   Specular=Constant() {r=0.5}
   Roughness=NamedRerouteUsage()
   EmissiveColor=MaterialFunctionCall(
    LDR Input=Clamp(None=@n10)
    HDR Dynamic Range=MaterialFunctionCall!(Low=@n11, High=@n11, Epic=@n11, --------, Mobile=@n12, Medium=@n12) {materialFunction=/FortniteBaseMaterials/Functions/MF_QualitySwitch_Scalar}.Result
    HDR Intensity Multiplier=ScalarParameter() {defaultValue=10, group=SL_S30_ExtraDetail, parameterName=HDRIntensity}
   ) {materialFunction=&DeriveHDRfromLDR}.RemappedRange
   Metallic=@n3
  ) {attributeSetTypes [
    #1
    #3
    #4
    #2
    #5
   ]}
  n14 = FunctionInput() {inputName=TechArtMask, sortPriority=2}
  n15 = SmoothStep(
   Min=ScalarParameter() {defaultValue=0.6, group=SL_S30_ExtraDetail, parameterName=ED_ClippingMask}
   Value=ChannelMaskParameter(None=NamedRerouteDeclaration(None=@n14) {name=TAMask, nodeColor=<rgba 0 1 0.2734 1>}) {group=SL_S30_ExtraDetail, parameterName=ExtraDetailMask}.R
  )
  FunctionOutput(None=StaticSwitchParameter(True=BlendMaterialAttributes(A=@n1, B=@n13, Alpha=@n15), False=NamedRerouteUsage()) {defaultValue=true, group=S30 Superleveling, parameterName=UseExtraDetail})
  NamedRerouteDeclaration(None=FunctionInput() {inputType=FunctionInput_MaterialAttributes}) {name=InMaterialAttributes, nodeColor=<rgba 1 0.3047 0 1>}
  n16 = TextureSample(UVs=Multiply(A=TextureCoordinate())) {samplerSource=SSM_Wrap_WorldGroupSettings, texture=/Game/Characters/Player/TechArt/Superlevels/S30/Textures/T_S30_Superlevel_T1_Hexagons}
  n17 = Power(Base=MaterialFunctionCall(In=@n16.B, Contrast=Constant()) {materialFunction=&CheapContrast}.Result) {constExponent=1}
  NamedRerouteDeclaration(None=LinearInterpolate(A=ScalarParameter() {defaultValue=0.2, group=SL_S30_CarbonFiber, parameterName=ED_Roughness}, Alpha=Saturate(None=@n17)) {constB=0.5}) {name=Roughness, nodeColor=<rgba 1 0 0.2109 1>}
  NamedRerouteDeclaration(None=FunctionInput(Preview=Constant() {r=1}) {inputName=Mask, sortPriority=1}) {name=MaskMap, nodeColor=<rgba 1 0 0.3047 1>}
 ]}
