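# Material function for the Season 30 (S30) player superlevel look, blending an animated blue hexagon surface over input material attributes through a channel-selected TechArtMask, behind the UseBlueHexagon switch.
# A tiled hexagon texture supplies a per-tile hash for sine-driven idle displacement, which feeds a BumpOffset parallax resample.
# That resample shapes the emissive color gradient.
# The graph also builds a Fresnel outline and two time-scrolled sweeps along pre-skinned local height.
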MaterialFunction /Game/Characters/Player/TechArt/Superlevels/S30/MaterialFunctions/MF_SL_S30_Hexagons
graph {nodes [
  n1 = NamedRerouteUsage()
  n2 = SetMaterialAttributes!(
   MaterialAttributes
   BaseColor=Constant()
   EmissiveColor=NamedRerouteUsage()
   Specular=Constant() {r=0.5}
   Roughness=Multiply(A=ScalarParameter() {defaultValue=1, group=SL_S30_Hexagon, parameterName=HexRoughness}, B=NamedRerouteUsage())
   Metallic=Constant()
  ) {attributeSetTypes [
    #1
    #2
    #3
    #4
    #5
   ]}
  n3 = Saturate(None=ChannelMaskParameter(None=FunctionInput() {inputName=TechArtMask, sortPriority=1}) {group=SL_S30_Hexagon, parameterName=HexagonMask}.R)
  FunctionOutput(None=StaticSwitchParameter(True=BlendMaterialAttributes(A=@n1, B=@n2, Alpha=@n3), False=NamedRerouteUsage()) {group=S30 Superleveling, parameterName=UseBlueHexagon})
  n4 = Multiply(
   A=Multiply(A=TextureCoordinate(), B=ScalarParameter() {defaultValue=17, group=SL_S30_Hexagon, parameterName=Hex_Tiling}) {constB=5}
   B=ComponentMask(None=VectorParameter() {group=SL_S30_Hexagon, parameterName=Hex_TilingRatio, defaultValue=<rgba 1 1 0 1>}.RGB) {g=true, r=true}
  )
  n5 = TextureSample(UVs=@n4) {samplerSource=SSM_Wrap_WorldGroupSettings, texture=/Game/Characters/Player/TechArt/Superlevels/S30/Textures/T_S30_Superlevel_T1_Hexagons}
  InverseLinearInterpolate(A=@n5.R, Value=NamedRerouteUsage())
  MaterialFunctionCall() {materialFunction=&BoundingBoxBased_0-1_UVW}
  n6 = SmoothStep(Max=ScalarParameter() {defaultValue=0.3, parameterName=Hex_Sharpness}, Value=@n5.G) {constMax=0.3}
  n7 = Time()
  n8 = ScalarParameter() {defaultValue=0.3, group=SL_S30_Hexagon, parameterName=HexIdleDisplacementSpeed}
  n9 = Subtract(
   A=NamedRerouteDeclaration(None=Multiply(A=@n7, B=@n8)) {name=Time, nodeColor=<rgba 0.9219 0 1 1>}
   B=NamedRerouteDeclaration(None=@n5.R) {name=Hash, nodeColor=<rgba 0.4062 1 0 1>}
  )
  n10 = Multiply(
   A=NamedRerouteDeclaration(None=Sine(None=@n9)) {name=IdleDisplacement, nodeColor=<rgba 0.2187 1 0 1>}
   B=ScalarParameter() {defaultValue=0.75, group=SL_S30_Hexagon, parameterName=HexIdleDisplacementAmount}
  )
  n11 = Multiply(A=ConstantBiasScale(None=NamedRerouteUsage()) {bias=-0.5, scale=2}) {constB=1.5}
  n12 = Subtract(A=Subtract(A=@n6, B=Multiply(A=@n6, B=@n10)), B=Multiply(A=@n6, B=@n11))
  Max(A=@n12)
  Min(A=@n12)
  n13 = TextureSample(UVs=BumpOffset(Coordinate=@n4, Height=Clamp(None=@n12) {minDefault=-1})) {samplerSource=SSM_Wrap_WorldGroupSettings, texture=&T_S30_Superlevel_T1_Hexagons}
  n14 = Multiply(A=Subtract(A=@n13.G, B=Multiply(A=@n13.B, B=@n13.B)), B=@n13.B)
  n15 = Saturate(None=Power(Base=OneMinus(None=@n14)) {constExponent=3})
  n16 = NamedRerouteUsage()
  Multiply(A=Add(A=LinearInterpolate(Alpha=@n15) {constA=0.2}, B=Multiply(A=@n16) {constB=0.2}), B=@n16)
  NamedRerouteDeclaration(None=FunctionInput() {inputType=FunctionInput_MaterialAttributes}) {name=InMaterialAttributes, nodeColor=<rgba 1 0.3047 0 1>}
  n17 = TextureSample(UVs=TextureCoordinate()) {samplerSource=SSM_Wrap_WorldGroupSettings, texture=&T_S30_Superlevel_T1_Hexagons}
  n18 = Constant() {r=0.2}
  NamedRerouteDeclaration(None=Power(Base=MaterialFunctionCall(In=@n17.B, Contrast=@n18) {materialFunction=&CheapContrast}.Result) {constExponent=1}) {name=Roughness, nodeColor=<rgba 1 0 0.2109 1>}
  n19 = SmoothStep(Max=ScalarParameter() {defaultValue=0.3, parameterName=Hex_Sharpness}, Value=@n13.G) {constMax=0.3}
  n20 = NamedRerouteUsage()
  n21 = OneMinus(None=Sine(None=Subtract(A=@n13.R, B=@n20)))
  n22 = ScalarParameter() {defaultValue=0.1, group=SL_S30_Hexagon, parameterName=HexTileDepthColor}
  n23 = Multiply(A=Add(A=OneMinus(None=@n19), B=Multiply(A=@n21, B=@n22) {constB=0.1})) {constB=0.5}
  n24 = Multiply(A=@n11, B=ScalarParameter() {defaultValue=0.05, group=SL_S30_Hexagon, parameterName=HexSheenIntensity}) {constB=0.2}
  n25 = LinearInterpolate(A=Clamp(None=Subtract(A=@n23, B=@n24)), Alpha=NamedRerouteUsage())
  n26 = ScalarParameter() {defaultValue=1, group=SL_S30_Hexagon, parameterName=HexOverallIntensity}
  NamedRerouteDeclaration(None=LinearInterpolate(
    A=VectorParameter() {group=SL_S30_Hexagon, parameterName=HexColor, defaultValue=<rgba 0 0.2026 0.7344 1>}.RGB
    B=VectorParameter() {group=SL_S30_Hexagon, parameterName=HexHighlightColor, defaultValue=<rgba 0 2.3148 14 1>}.RGB
    Alpha=Multiply(A=@n25, B=@n26)
   )) {name=Emissive, nodeColor=<rgba 1 0.5859 0 1>}
  n27 = ScalarParameter() {group=SL_S30_Hexagon, parameterName=OutlineThreshhold}
  n28 = NamedRerouteUsage()
  n29 = Power(
   Base=Fresnel(ExponentIn=Constant() {r=1}) {baseReflectFraction=0, exponent=0.5}
   Exp=Add(A=ScalarParameter() {defaultValue=1.25, group=SL_S30_Hexagon, parameterName=OutlinePower}, B=Multiply(A=@n28) {constB=1.5})
  )
  NamedRerouteDeclaration(None=Multiply(
    A=SmoothStep(Min=@n27, Value=@n29) {constMin=0.1}
    B=ScalarParameter() {defaultValue=2, group=SL_S30_Hexagon, parameterName=OutlineIntensity}
   )) {name=Fresnel, nodeColor=<rgba 0.9688 1 0 1>}
  n30 = LocalPosition() {localOrigin=InstancePreSkinning}
  n31 = Divide(A=ComponentMask(None=VertexInterpolator(VS=@n30.XYZ).PS) {b=true}) {constB=100}
  n32 = ScalarParameter() {defaultValue=0.1, parameterName=HexSweepSpeed}
  n33 = Add(
   A=NamedRerouteDeclaration(None=Divide(A=@n31) {constB=1}) {name=Height, nodeColor=<rgba 1 0.1875 0 1>}
   B=Multiply(A=Time(), B=@n32) {constB=0.1}
  )
  n34 = Power(Base=Abs(None=Sine(None=@n33))) {constExponent=10}
  n35 = Divide(A=@n31)
  n36 = Time()
  n37 = Divide(A=@n32)
  n38 = Sine(None=Add(A=Add(A=@n35) {constB=15}, B=Multiply(A=@n36, B=@n37) {constB=0.05}))
  n39 = Saturate(None=Power(Base=Abs(None=@n38)) {constExponent=10})
  n40 = Add(
   A=NamedRerouteDeclaration(None=Saturate(None=@n34)) {name=MainSweep, nodeColor=<rgba 1 0.0234 0 1>}
   B=Multiply(A=NamedRerouteDeclaration(None=@n39) {name=SecondarySweep, nodeColor=<rgba 0.8984 0 1 1>}) {constB=0.5}
  )
  n41 = NamedRerouteUsage()
  NamedRerouteDeclaration(None=Multiply(A=NamedRerouteDeclaration(None=@n40) {name=FullSweep, nodeColor=<rgba 0.8281 0 1 1>}, B=LinearInterpolate(Alpha=@n41) {constA=0.5})) {name=TopdownGradientHash, nodeColor=<rgba 0 0.5781 1 1>}
  OneMinus(None=@n11)
 ]}
