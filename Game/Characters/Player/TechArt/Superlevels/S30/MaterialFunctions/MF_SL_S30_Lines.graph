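# Material function for the Season 30 player superlevel "Lines" style, blended over incoming material attributes through a channel-selected TechArtMask behind the UseLines static switch.
# Two opposingly distorted samples of a curves texture form animated lines, using panning UVs or optional triplanar warping.
# Emissive combines tinted colors passed through a scanline function with a fresnel color term.
# A time-driven sine gradient uses pre-skinned height, and a hexagon texture derives roughness.
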MaterialFunction /Game/Characters/Player/TechArt/Superlevels/S30/MaterialFunctions/MF_SL_S30_Lines
graph {nodes [
  n1 = NamedRerouteUsage()
  n2 = Constant() {r=1.4}
  n3 = Constant() {r=0.4}
  n4 = LinearInterpolate(
   A=LinearInterpolate(
    A=VectorParameter() {group=SL_S30_Lines, parameterName=L_BGColor2, defaultValue=<rgba 0.6661 0.0403 0.8958 1>}.RGB
    B=VectorParameter() {group=SL_S30_Lines, parameterName=L_BGColor1, defaultValue=<rgba 0.0208 0.1842 0.526 1>}.RGB
    Alpha=NamedRerouteUsage()
   )
   B=VectorParameter() {group=SL_S30_Lines, parameterName=L_ColorTint, defaultValue=<rgba 0 0.6256 2 1>}.RGB
   Alpha=NamedRerouteUsage()
  )
  n5 = NamedRerouteUsage()
  n6 = VectorParameter() {group=SL_S30_Lines, parameterName=L_FresnelColor, defaultValue=<rgba 0.4622 0.5404 0.5885 1>}
  n7 = SetMaterialAttributes!(
   MaterialAttributes
   BaseColor=Constant()
   Specular=Constant() {r=0.5}
   Roughness=NamedRerouteUsage()
   Metallic=ScalarParameter() {group=SL_S30_Lines, parameterName=L_Metalness}
   EmissiveColor=Add(
    A=MaterialFunctionCall!(ScanlineIntensity=@n2, ScanlineThickness, ScanlineScale, ScanNegativeIntensity=@n3, UVs, In=@n4) {materialFunction=&/MF_SL_S30_Scanlines}.Result
    B=Multiply(A=@n5, B=@n6.RGB)
   )
  ) {attributeSetTypes [
    #1
    #2
    #3
    #4
    #5
   ]}
  n8 = ChannelMaskParameter(None=NamedRerouteDeclaration(None=FunctionInput() {inputName=TechArtMask, sortPriority=2}) {name=TAMask, nodeColor=<rgba 0 1 0.2734 1>}) {group=SL_S30_Lines, parameterName=LinesMask}
  FunctionOutput(None=StaticSwitchParameter(True=BlendMaterialAttributes(A=@n1, B=@n7, Alpha=@n8.R), False=NamedRerouteUsage()) {group=S30 Superleveling, parameterName=UseLines})
  GetMaterialAttributes(None=NamedRerouteUsage()) {attributeGetTypes [#1, #5]}
  NamedRerouteDeclaration(None=FunctionInput() {inputType=FunctionInput_MaterialAttributes}) {name=InMaterialAttributes, nodeColor=<rgba 1 0.3047 0 1>}
  n9 = TextureSample(UVs=TextureCoordinate()) {samplerSource=SSM_Wrap_WorldGroupSettings, texture=/Game/Characters/Player/TechArt/Superlevels/S30/Textures/T_S30_Superlevel_T1_Hexagons}
  n10 = MaterialFunctionCall(In=@n9.B, Contrast=Constant() {r=0.2}) {materialFunction=&CheapContrast}
  NamedRerouteDeclaration(None=Multiply(A=Power(Base=@n10.Result) {constExponent=1}, B=ScalarParameter() {defaultValue=1, group=SL_S30_Lines, parameterName=L_Roughness})) {name=Roughness, nodeColor=<rgba 1 0 0.2109 1>}
  n11 = Time()
  n12 = ScalarParameter() {defaultValue=0.05, group=SL_S30_Lines, parameterName=L_GradientPanningSpeed}
  n13 = PreSkinnedPosition()
  n14 = ComponentMask(None=ShaderStageSwitch(PixelShader=VertexInterpolator(VS=@n13).PS, VertexShader=@n13)) {b=true}
  n15 = ScalarParameter() {defaultValue=70, group=SL_S30_Lines, parameterName=L_GradientSize}
  n16 = Sine(None=Add(A=Multiply(A=@n11, B=@n12) {constB=0.15}, B=Divide(A=@n14, B=@n15) {constB=-125}))
  NamedRerouteDeclaration(None=Power(Base=Abs(None=@n16), Exp=ScalarParameter() {defaultValue=1, group=SL_S30_Lines, parameterName=L_GradientSweeoSmoothness}) {constExponent=8}) {name=TopdownGradient, nodeColor=<rgba 0 1 0.3906 1>}
  NamedRerouteDeclaration(None=FunctionInput() {inputName=Fresnel, inputType=FunctionInput_Scalar, sortPriority=1}) {name=InFresnel, nodeColor=<rgba 1 0 0.4453 1>}
  n17 = Fresnel(ExponentIn=ScalarParameter() {defaultValue=2, group=SL_S30_Lines, parameterName=L_FresnelPower}) {baseReflectFraction=0, exponent=1}
  n18 = ScalarParameter() {defaultValue=0.5, group=SL_S30_Lines, parameterName=L_FresnelIntensity}
  NamedRerouteDeclaration(None=Saturate(None=Multiply(A=@n17, B=@n18))) {name=OutFresnel, nodeColor=<rgba 0.3125 0 1 1>}
  NamedRerouteDeclaration(None=FunctionInput(Preview=Constant() {r=1}) {inputName=Mask, sortPriority=1}) {name=MaskMap, nodeColor=<rgba 1 0 0.3047 1>}
  n19 = TextureObject() {texture=/Game/Characters/Player/TechArt/Superlevels/S30/Textures/T_S30_Superlevel_T3_Curves}
  n20 = Constant3Vector() {constant=<rgba 0.02 0 0.2 0>}
  n21 = Constant() {r=0.75}
  n22 = NamedRerouteUsage()
  n23 = MaterialFunctionCall!(
   MainTexture=@n19
   Local Position
   Triplanner Speed=@n20
   TriPlanner AxisBlendStrength
   TriplannarTextureScale
   Local Normal
   NoiseTiling=@n21
   NoiseDeformIntensity=@n22
   NoiseTexture=@n19
   NoiseChannel
  ) {materialFunction=&/MF_SL_S30_TriplanarWarp}
  n24 = MaterialFunctionCall(
   MainTexture=@n19
   Triplanner Speed=@n20
   TriplannarTextureScale=@n21
   NoiseTiling=@n21
   NoiseDeformIntensity=Multiply(A=@n22) {constB=-1}
   NoiseTexture=@n19
  ) {materialFunction=&MF_SL_S30_TriplanarWarp}
  n25 = MaterialFunctionCall(
   UVs=Multiply(A=TextureCoordinate(), B=ScalarParameter() {defaultValue=3, group=SL_S30_Lines, parameterName=L_LinesTiling})
   "Rotation Angle (0-1)"=ScalarParameter() {group=SL_S30_Lines, parameterName=L_LinesRotation}
  ) {materialFunction=&CustomRotator}
  n26 = Add(
   A=Multiply(A=@n25.Rotated Values) {constB=0.75}
   B=ComponentMask(None=VectorParameter() {group=SL_S30_Lines, parameterName=L_NoiseOffset, defaultValue=<rgba 0.6 0 0 1>}.RGB) {g=true, r=true}
  )
  n27 = Multiply(
   A=Subtract(A=TextureSample(UVs=@n26) {texture=&T_S30_Superlevel_T3_Curves}.G) {constB=0.5}
   B=NamedRerouteDeclaration(None=ScalarParameter() {defaultValue=0.3, group=SL_S30_Lines, parameterName=L_DeformIntensity}) {name=DeformIntensity, nodeColor=<rgba 0 1 0.8125 1>}
  ) {constB=0.4}
  n28 = ComponentMask(None=NamedRerouteDeclaration(None=VectorParameter() {group=SL_S30_Lines, parameterName=L_LinesDiretion, defaultValue=<rgba 0 0.25 0 1>}.RGB) {name=LinesDirection, nodeColor=<rgba 1 0.2578 0 1>}) {g=true, r=true}
  n29 = Panner(Coordinate=Add(A=@n25.Rotated Values, B=@n27), Speed=@n28) {speedX=1, speedY=0.075}
  n30 = Panner(Coordinate=Subtract(A=Multiply(A=@n25.Rotated Values) {constB=0.75}, B=@n27), Speed=@n28) {speedX=-0.02, speedY=-0.025}
  n31 = StaticSwitchParameter(
   True=Add(A=ComponentMask(None=@n23.XY) {a=true}, B=ComponentMask(None=@n24.XY) {a=true})
   False=Add(
    A=TextureSample(UVs=@n29) {texture=&T_S30_Superlevel_T3_Curves}.A
    B=TextureSample(UVs=@n30) {texture=&T_S30_Superlevel_T3_Curves}.A
   )
  ) {group=SL_S30_Lines, parameterName=L_UseTriplanar}
  NamedRerouteDeclaration(None=Saturate(None=SmoothStep(Value=@n31) {constMin=0.05})) {name=Lines, nodeColor=<rgba 0.6641 0 1 1>}
 ]}
