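# Material function for the S30 player superlevel "Squares" look, blended over incoming material attributes by a TechArtMask channel behind the UseSquares static switch.
# Squares use stabilized screen-space UVs distorted by a glitch function, combining three evaluations into color-split squares with a tint.
# These composite over a two-color background with a Fresnel rim broken up by panning curve textures.
# Scanlines modulate the result into emissive over black base color, with hexagon-textured roughness.
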MaterialFunction /Game/Characters/Player/TechArt/Superlevels/S30/MaterialFunctions/MF_SL_S30_SquareEffect
graph {nodes [
  n1 = NamedRerouteUsage()
  n2 = Power(Base=OneMinus(None=NamedRerouteUsage()), Exp=ScalarParameter() {defaultValue=16, group=SL_S30_Squares, parameterName=S_ColorContrast}) {constExponent=8}
  n3 = NamedRerouteUsage()
  n4 = LinearInterpolate(
   A=LinearInterpolate(
    A=VectorParameter() {group=SL_S30_Squares, parameterName=S_BGColor1, defaultValue=<rgba 0.0163 0.0082 0.0833 1>}.RGB
    B=VectorParameter() {group=SL_S30_Squares, parameterName=S_BGColor2, defaultValue=<rgba 0.1163 0 0.3542 1>}.RGB
    Alpha=Saturate(None=@n2)
   )
   B=@n3
   Alpha=Saturate(None=@n3)
  )
  n5 = Multiply(A=NamedRerouteUsage(), B=VectorParameter() {group=SL_S30_Squares, parameterName=S_FresnelColor, defaultValue=<rgba 0 0.5235 0.651 1>}.RGB)
  n6 = SetMaterialAttributes!(
   MaterialAttributes=GetMaterialAttributes(None=NamedRerouteUsage()) {attributeGetTypes [#1, #2]}.MaterialAttributes
   BaseColor=Constant()
   Specular=Constant() {r=0.5}
   Roughness=NamedRerouteUsage()
   EmissiveColor=MaterialFunctionCall!(
    ScanlineIntensity=ScalarParameter() {defaultValue=3.5, group=SL_S30_Squares, parameterName=S_ScanlineMax}
    ScanlineThickness
    ScanlineScale
    ScanNegativeIntensity=ScalarParameter() {defaultValue=0.2, group=SL_S30_Squares, parameterName=S_ScanlineMin}
    UVs
    In=Add(A=@n4, B=@n5)
   ) {materialFunction=&/MF_SL_S30_Scanlines}.Result
   Metallic=ScalarParameter() {group=SL_S30_Squares, parameterName=S_Metalness}
  ) {attributeSetTypes [
    #1
    #3
    #2
    #4
    #5
   ]}
  n7 = ChannelMaskParameter(None=NamedRerouteDeclaration(None=FunctionInput() {inputName=TechArtMask, sortPriority=2}) {name=TAMask, nodeColor=<rgba 0 1 0.2734 1>}) {group=SL_S30_Squares, parameterName=SquaresMask}
  FunctionOutput(None=StaticSwitchParameter(True=BlendMaterialAttributes(A=@n1, B=@n6, Alpha=@n7.R), False=NamedRerouteUsage()) {group=S30 Superleveling, parameterName=UseSquares})
  NamedRerouteDeclaration(None=FunctionInput() {inputType=FunctionInput_MaterialAttributes}) {name=InMaterialAttributes, nodeColor=<rgba 1 0.3047 0 1>}
  n8 = TextureSample(UVs=Multiply(A=TextureCoordinate())) {samplerSource=SSM_Wrap_WorldGroupSettings, texture=/Game/Characters/Player/TechArt/Superlevels/S30/Textures/T_S30_Superlevel_T1_Hexagons}
  n9 = Power(Base=MaterialFunctionCall(In=@n8.B, Contrast=Constant()) {materialFunction=&CheapContrast}.Result) {constExponent=1}
  NamedRerouteDeclaration(None=LinearInterpolate(A=ScalarParameter() {defaultValue=0.2, group=SL_S30_Squares, parameterName=S_Roughness}, Alpha=Saturate(None=@n9))) {name=Roughness, nodeColor=<rgba 1 0 0.2109 1>}
  NamedRerouteDeclaration(None=FunctionInput(Preview=Fresnel() {baseReflectFraction=0, exponent=1}) {inputName=Fresnel, inputType=FunctionInput_Scalar, sortPriority=1}) {name=InFresnel, nodeColor=<rgba 1 0 0.4453 1>}
  NamedRerouteDeclaration(None=FunctionInput(Preview=Constant() {r=1}) {inputName=Mask, sortPriority=1}) {name=MaskMap, nodeColor=<rgba 1 0 0.3047 1>}
  n10 = LocalPosition()
  n11 = ComponentMask(None=VertexInterpolator(VS=Divide(A=@n10.XYZ) {constB=30}).PS) {b=true, r=true}
  n12 = Panner(Coordinate=@n11) {speedY=-0.1}
  n13 = Panner(Coordinate=Multiply(A=@n11) {constB=1.5}) {speedY=0.15}
  n14 = Saturate(None=Add(
    A=TextureSample(UVs=@n12) {texture=/Game/Characters/Player/TechArt/Superlevels/S30/Textures/T_S30_Superlevel_T3_Curves}.G
    B=TextureSample(UVs=@n13) {texture=&T_S30_Superlevel_T3_Curves}.G
   ))
  n15 = Saturate(None=Fresnel(ExponentIn=ScalarParameter() {defaultValue=0.5, group=SL_S30_Squares, parameterName=S_FresnelPower}) {baseReflectFraction=0, exponent=1})
  n16 = ScalarParameter() {defaultValue=0.9, group=SL_S30_Squares, parameterName=S_FresnelIntensity}
  n17 = Saturate(None=InverseLinearInterpolate(A=Multiply(A=@n14) {constB=0.5}, Value=Multiply(A=@n15, B=@n16)))
  NamedRerouteDeclaration(None=Multiply(A=@n17, B=@n17)) {name=OutFresnel, nodeColor=<rgba 0.3125 0 1 1>}
  n18 = Constant3Vector() {constant=<rgba 0 0 110 1>}
  n19 = Add(
   A=Multiply(
    A=MaterialFunctionCall!(ObjectPivotLocation, ObjectPositionOffset=@n18) {materialFunction=/Game/Characters/Player/Common/MaterialFunctions/Superlevel/ImprovedStabilizedScreenSpaceUVs}.RawUVs
    B=ScalarParameter() {defaultValue=0.015, group=SL_S30_Squares, parameterName=SQ_SquaresScale}
   ) {constB=0.015}
   B=ComponentMask(None=VectorParameter() {group=SL_S30_Squares, parameterName=SQ_SquaresOffset, defaultValue=<rgba 0 0 0 1>}.RGB) {g=true, r=true}
  )
  n20 = TextureSample(UVs=@n19) {
   samplerSource=SSM_Wrap_WorldGroupSettings
   samplerType=SAMPLERTYPE_LinearColor
   texture=/Game/Characters/Player/TechArt/Superlevels/S30/Textures/T_S30_Superlevel_T3_Squares
  }
  n21 = MaterialFunctionCall!(In=@n19, Frequency, IntensityMultiplier, TimeOffset=@n20.R) {materialFunction=&/MF_SL_S30_Glitch}
  n22 = MaterialFunctionCall!(UV=Add(A=@n21.GlitchedUVs, B=@n21.GlitchedUVs)) {materialFunction=&/MF_SL_S30_Squares}
  n23 = MaterialFunctionCall(UV=Subtract(A=@n21.GlitchedUVs, B=@n21.GlitchedUVs)) {materialFunction=&MF_SL_S30_Squares}
  NamedRerouteDeclaration(None=LinearInterpolate(
    A=MaterialFunctionCall(R=@n22.Result, G=@n23.Result) {materialFunction=&AppendMany}.RGB
    B=VectorParameter() {group=SL_S30_Squares, parameterName=S_SquareColor, defaultValue=<rgba 0 0.3347 1 1>}.RGB
    Alpha=MaterialFunctionCall(UV=@n21.GlitchedUVs) {materialFunction=&MF_SL_S30_Squares}.Result
   )) {name=Squares, nodeColor=<rgba 1 0.6094 0 1>}
 ]}
