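# Parent material for Season 30 (S30) superlevel character skins, layering effect looks over the shared 2023 character parent.
# Between the PreFX and PostFX functions, material attributes pass serially through emissive skin, base, white leather, hexagon, black leather, line, square, extra detail and carbon fiber functions.
# Two tech-art textures supply masks, with an optionally clip-refined detail channel cutting out the main mask.
# A static debug switch can preview masks as emissive.
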Material /Game/Characters/Player/TechArt/Superlevels/S30/M_Superleveling_S30_Parent
graph {
 properties {
  bHasPixelAnimation=true
  bUseMaterialAttributes=true
  bUsedWithInstancedStaticMeshes=true
  bUsedWithMeshDeformer=true
  bUsedWithMorphTargets=true
  bUsedWithNanite=true
  bUsedWithSkeletalMesh=true
  pixelDepthOffsetMode=PDOM_Legacy
  shadingModel=MSM_FromMaterialExpression
  usedShadingModels=Default Lit | Subsurface | Clear Coat | Subsurface Profile | Cloth
 }
 nodes [
  n1 = MaterialFunctionCall() {materialFunction=&MF_CharacterParent_2023_PreFX}
  MaterialFunctionCall(Material=NamedRerouteUsage(), M=@n1.M, Pre Skinned Local Position=@n1.M, Distance Blend=@n1.M, Fresnel=@n1.M) {materialFunction=&MF_CharacterParent_2023_PostFX}
  n2 = NamedRerouteUsage()
  n3 = MaterialFunctionCall!(In=@n1.Result, Mask=@n2) {materialFunction=/Game/Characters/Player/TechArt/Superlevels/S30/MaterialFunctions/MF_SL_S30_EmissiveSkin}
  n4 = NamedRerouteUsage()
  n5 = NamedRerouteUsage()
  n6 = MaterialFunctionCall!(
   In=MaterialFunctionCall!(
    In=MaterialFunctionCall!(In=@n3.Result, Mask=@n1.M, TechArtDetailMask=@n4, TechArtMask=@n5) {materialFunction=/Game/Characters/Player/TechArt/Superlevels/S30/MaterialFunctions/MF_SL_S30_Base}.Result
    Fresnel=@n1.Fresnel
    Mask=@n1.Fresnel
    SL2Mask=@n2
    TechArtMask=@n5
   ) {materialFunction=/Game/Characters/Player/TechArt/Superlevels/S30/MaterialFunctions/MF_SL_S30_WhiteLeather}.Result
   TechArtMask=@n5
  ) {materialFunction=/Game/Characters/Player/TechArt/Superlevels/S30/MaterialFunctions/MF_SL_S30_Hexagons}
  n7 = MaterialFunctionCall!(
   In=MaterialFunctionCall!(
    In=MaterialFunctionCall!(In=@n6.Result, Mask=@n1.M, SL2Mask=@n2, TechArtMask=@n5) {materialFunction=/Game/Characters/Player/TechArt/Superlevels/S30/MaterialFunctions/MF_SL_S30_BlackLeather}.Result
    Fresnel=@n1.Fresnel
    Mask=@n1.Fresnel
    TechArtMask=@n5
   ) {materialFunction=/Game/Characters/Player/TechArt/Superlevels/S30/MaterialFunctions/MF_SL_S30_Lines}.Result
   Fresnel=@n1.Fresnel
   Mask=@n1.Fresnel
   TechArtMask=@n5
  ) {materialFunction=/Game/Characters/Player/TechArt/Superlevels/S30/MaterialFunctions/MF_SL_S30_SquareEffect}
  NamedRerouteDeclaration(None=MaterialFunctionCall!(
    In=MaterialFunctionCall!(In=@n7.Result, Mask=@n1.M, TechArtMask=@n4) {materialFunction=/Game/Characters/Player/TechArt/Superlevels/S30/MaterialFunctions/MF_SL_S30_ExtraDetail}.Result
    Mask=@n1.M
    TechArtMask=@n5
   ) {materialFunction=/Game/Characters/Player/TechArt/Superlevels/S30/MaterialFunctions/MF_SL_S30_CarbonFiber}.Result) {name=InMaterialAttributes, nodeColor=<rgba 0 0.2266 1 1>}
  n8 = TextureSampleParameter2D() {group=S30 Superleveling, parameterName=TAMaskSL, samplerType=SAMPLERTYPE_Masks, texture=&T_BlackMask}
  n9 = TextureSampleParameter2D() {
   group=S30 Superleveling
   parameterName=TAMaskSLDetail
   samplerType=SAMPLERTYPE_Masks
   texture=&T_BlackMask
  }
  n10 = ScalarParameter() {defaultValue=0.4, group=S30 Superleveling, parameterName=SL_MaskClipping}
  n11 = LinearInterpolate(
   A=@n9.R
   B=Saturate(None=InverseLinearInterpolate(B=@n10, Value=@n9.R))
   Alpha=ScalarParameter() {defaultValue=1, group=S30 Superleveling, parameterName=SL_UseRefinedMask}
  )
  NamedRerouteDeclaration(None=LinearInterpolate(A=@n8.RGB, Alpha=@n11) {constB=0}) {name=TASLMask, nodeColor=<rgba 1 0 0.2109 1>}
  n12 = Constant()
  n13 = Constant() {r=1}
  n14 = LinearInterpolate(
   A=NamedRerouteUsage()
   B=NamedRerouteUsage()
   Alpha=ScalarParameter() {group=S30 Superleveling, parameterName=DetailMaskPreview, sliderMax=1}
  )
  NamedRerouteDeclaration(None=StaticSwitchParameter(
    True=MakeMaterialAttributes!(
     BaseColor=@n12
     Metallic=@n12
     Specular=@n12
     Roughness=@n13
     Anisotropy
     EmissiveColor=@n14
     Opacity
     OpacityMask
     Normal
     Tangent
     WorldPositionOffset
     SubsurfaceColor
     ClearCoat
     ClearCoatRoughness
     AmbientOcclusion
     Refraction
     CustomizedUVs_0
     CustomizedUVs_1
     CustomizedUVs_2
     CustomizedUVs_3
     CustomizedUVs_4
     CustomizedUVs_5
     CustomizedUVs_6
     CustomizedUVs_7
     PixelDepthOffset
     ShadingModel
     Displacement
    )
    False=NamedRerouteUsage()
   ) {group=S30 Superleveling, parameterName=DebugTAMask}) {name=OutMaterialAttributes, nodeColor=<rgba 0 0.4141 1 1>}
  NamedRerouteDeclaration(None=@n11) {name=RefinedTASLDetailMask, nodeColor=<rgba 1 0 0.2109 1>}
  NamedRerouteDeclaration(None=@n9.RGB) {name=SL2TAMask, nodeColor=<rgba 0 1 0.8594 1>}
 ]
}
